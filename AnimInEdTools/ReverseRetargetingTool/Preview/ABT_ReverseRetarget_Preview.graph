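# Template animation blueprint that previews reverse retargeting in the AnimInEd tools plugin, with no target skeleton set.
# The AnimGraph is a single Retarget Pose From Mesh node that feeds the output pose.
# Its retargeter comes from the IKRetargeter Asset variable, which defaults to IKR_ReverseRetarget.
# The EventGraph holds only unconnected default nodes.
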AnimBlueprint /AnimInEdTools/ReverseRetargetingTool/Preview/ABT_ReverseRetarget_Preview
target_skeleton=null
graphs [
 AnimGraph {nodes [Output Pose(Result=Retarget Pose From Mesh!(IKRetargeterAsset:IKRetargeter Object Reference=Get IKRetargeter Asset() -> (IKRetargeter Asset:IKRetargeter Object Reference)) -> (Pose:Pose Link Structure) {is=title AnimGraphNode_RetargetPoseFromMesh})]}
 EventGraph {nodes [
   TryGetPawnOwner()
   Event BlueprintUpdateAnimation()
  ]}
]
variables [IKRetargeter Asset {type=IKRetargeter, default=/AnimInEdTools/ReverseRetargetingTool/IKRetargeter/IKR_ReverseRetarget}]
